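# Physics asset for the M_MED_Emperor_Suit_Pack backpack cosmetic mesh.
# It holds a single body on the dny_ball bone, a zero-length capsule that acts as a sphere centered almost exactly on the bone.
# No constraints are defined, so the asset links this body to no other.
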PhysicsAsset /BRCosmetics/Accessories/FORT_Backpacks/Backpack_M_MED_Emperor_Suit/Meshes/M_MED_Emperor_Suit_Pack_Physics
bodies {dny_ball [{shapeType=Capsule, length=0, radius=12.1302, center=<0 0.0001 -0.0003>, rotation=<rot 0 24.0814 -96.2798>}]}
constraints []
