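# Animation Blueprint for the Emperor Suit backpack that adds physics motion and counter-spins its ball and text bones.
# A RigidBody node simulates in root bone space with overridden gravity, then two additive Modify Bone nodes apply yaw rotators.
# Each update, spinFunction advances both angles by opposite fixed steps plus a velocity term, subtracting 720 at that threshold.
# The graph never sets velocity, and deltaTimeRemap stays at 1, so spin is per-frame.
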AnimBlueprint /BRCosmetics/Accessories/FORT_Backpacks/Backpack_M_MED_Emperor_Suit/Meshes/M_MED_Emperor_Suit_Pack_AnimBP
target_skeleton=&/M_MED_Emperor_Suit_Pack_Skeleton
graphs [
 AnimGraph {nodes [
   n1 = RigidBody(ComponentPose=Local To Component(LocalPose=Local Space Ref Pose())) {
    bOverrideWorldGravity=true
    overridePhysicsAsset=&/M_MED_Emperor_Suit_Pack_Physics
    simulationSpace=BaseBoneSpace
    alphaBoolBlend {blendOption=Linear}
    baseBoneRef {boneName=root}
    componentAppliedLinearAccClamp=<2000 2000 2000>
    componentLinearAccScale=<1 1 1>
    componentLinearVelScale=<1 1 1>
    overrideWorldGravity {z=-980}
   }
   n2 = "Transform (Modify) Bone - Bone: dny_ball_text"(ComponentPose=@n1, Alpha=1, Translation="0,0,0", Rotation=Get text() -> (text:Rotator), Scale="1,1,1") {
    rotationMode=BMM_Additive
    rotationSpace=BCS_BoneSpace
    translationMode=BMM_Additive
    translationSpace=BCS_BoneSpace
    alphaBoolBlend {blendOption=Linear}
    boneToModify {boneName=dny_ball_text}
   }
   n3 = "Transform (Modify) Bone - Bone: dny_ball"(ComponentPose=@n2, Alpha=1, Translation="0,0,0", Rotation=Get globe() -> (globe:Rotator), Scale="1,1,1") {
    rotationMode=BMM_Additive
    rotationSpace=BCS_BoneSpace
    translationMode=BMM_Additive
    translationSpace=BCS_BoneSpace
    alphaBoolBlend {blendOption=Linear}
    boneToModify {boneName=dny_ball}
   }
   Output Pose(Result=Component To Local(ComponentPose=@n3)) {layerGroup=None}
  ]}
 EventGraph {nodes [
   Event BlueprintUpdateAnimation()
   > n1 = Sequence!(execute) -> (then_0, then_1, then_2:Exec)
   > Set deltaTimeRemap(deltaTimeRemap:double=FInterpTo(Current=Get deltaTimeRemap() -> (deltaTimeRemap:double), Target=1, DeltaTime=0, InterpSpeed=1)) -> (Output_Get:double)
   n2 = spinFunction!(execute=@n1.then_1, self, spin:double=Get textSpin() -> (textSpin:double), speed:double=0.75) -> (then, output:double)
   > n3 = Set textSpin(textSpin:double=@n2) -> (Output_Get:double)
   > Set text(text:Rotator=MakeRotator(Yaw=@n3)) -> (Output_Get:Rotator)
   n4 = spinFunction(execute=@n1.then_2, spin=Get globeSpin() -> (globeSpin:double), speed=-0.75)
   > n5 = Set globeSpin(globeSpin:double=@n4) -> (Output_Get:double)
   > Set globe(globe:Rotator=MakeRotator(Yaw=@n5)) -> (Output_Get:Rotator)
  ]}
 spinFunction {nodes [
   n1 = spinFunction!() -> (then, spin:double=~, speed:double=~) {is=title K2Node_FunctionEntry}
   > n2 = Set spinVar(spinVar:double=@n1) -> (Output_Get:double)
   > n3 = Branch(Condition="float >= float"(A=@n2, B=720) {is=title K2Node_CallFunction})
   > n4 = Set spinVar(spinVar=float - float(A=Get spinVar() -> (spinVar:double), B=720) {is=title K2Node_CallFunction})
   n5 = Get velocity() -> (velocity:double)
   n6 = Get velocityMult() -> (velocityMult:double)
   n7 = Get deltaTimeRemap()
   n8 = Set spinVar(
    execute=@n3.else
    spinVar=float + float!(A=Get spinVar(), B=float * float(A=@n1.speed, B=@n7), C:double=float * float(A=@n5, B=@n6)) -> (ReturnValue) {is=title K2Node_CommutativeAssociativeBinaryOperator}
   )
   Return Node!(execute=[@n4.then, @n8.then], output:double=@n8)
  ]}
]
variables [
 deltaTimeRemap {type=number, default=1}
 globe {type=Rotator}
 globeSpin {type=number}
 spinVar {type=number}
 text {type=Rotator}
 textSpin {type=number}
 velocity {type=number}
 velocityMult {type=number, default=0.5}
]
